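# HeistLoading is the loading screen widget for the Heist limited-time mode, cycling artwork slides with styled gameplay tips.
# It starts on a random slide, fades in from dark, and can embed an optional custom widget.
# Idle and transition animations alternate, with parallel arrays indexed by SlidePosition supplying each slide's image, text, icon and alignment.
# Two stacked images crossfade, adding a zoom only when the brushes differ, while the sheared tip panel wipes across.
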WidgetBlueprint /Game/UI/Foundation/Textures/LoadingScreens/LTM/HeistLoading
animations [
 SlideTransition {
  binding_count=1
  display_label=""
  length_seconds=0.75
  object=&self:SlideTransition
  sequence {
   bindings [Image_Loading_Incoming {
     id=#1
     possessed_class=Image
     tracks [
      {
       type=MovieSceneFloatTrack
       display_name=Render Opacity
       sections [{type=MovieSceneFloatSection, end_frame=45, start_frame=0, channels [None {type=MovieSceneScriptingFloatChannel, keys [[0,0],[44.999,1]]}]}]
      }
      {
       type=MovieScene2DTransformTrack
       sections [{
         type=MovieScene2DTransformSection
         end_frame=45
         start_frame=0
         channels [
          Translation.X {type=MovieSceneScriptingFloatChannel}
          Translation.Y {type=MovieSceneScriptingFloatChannel}
          Angle {type=MovieSceneScriptingFloatChannel}
          Scale.X {type=MovieSceneScriptingFloatChannel, keys [[0,1.1],[44.999,1]]}
          Scale.Y {type=MovieSceneScriptingFloatChannel, keys [[0,1.1],[44.999,1]]}
          Shear.X {type=MovieSceneScriptingFloatChannel}
          Shear.Y {type=MovieSceneScriptingFloatChannel}
         ]
        }]
      }
     ]
    }]
   display_rate {numerator=60}
   tracks [{type=MovieSceneEventTrack, sections [{type=MovieSceneEventSection}]}]
  }
 }
 Idle {
  binding_count=0
  display_label=""
  length_seconds=5.0259
  object=&self:Idle
  sequence {bindings [], tracks [{type=MovieSceneEventTrack, sections [{type=MovieSceneEventSection}]}]}
 }
 TipTransition {
  binding_count=4
  display_label=""
  length_seconds=0.75
  object=&self:TipTransition
  sequence {
   bindings [
    CommonTextBlock_Title {
     id=#2
     display_name=RichTextBlock_Title
     possessed_class=CommonTextBlock
     tracks [
      {
       type=MovieScene2DTransformTrack
       sections [{
         type=MovieScene2DTransformSection
         end_frame=44
         start_frame=0
         channels [
          Translation.X {type=MovieSceneScriptingFloatChannel, keys [[0,0],[20.535,740.0143],[27,-558.4147],[44.998,0]]}
          Translation.Y {type=MovieSceneScriptingFloatChannel}
          Angle {type=MovieSceneScriptingFloatChannel}
          Scale.X {type=MovieSceneScriptingFloatChannel}
          Scale.Y {type=MovieSceneScriptingFloatChannel}
          Shear.X {type=MovieSceneScriptingFloatChannel}
          Shear.Y {type=MovieSceneScriptingFloatChannel}
         ]
        }]
      }
      {
       type=MovieSceneFloatTrack
       display_name=Render Opacity
       sections [{
         type=MovieSceneFloatSection
         end_frame=45
         start_frame=0
         channels [None {type=MovieSceneScriptingFloatChannel, keys [[0,1],[20.535,0],[27,0],[44.997,1]]}]
        }]
      }
     ]
    }
    RichTextBlock_Tip {
     id=#3
     possessed_class=RichTextBlock
     tracks [
      {
       type=MovieScene2DTransformTrack
       sections [{
         type=MovieScene2DTransformSection
         end_frame=45
         start_frame=0
         channels [
          Translation.X {type=MovieSceneScriptingFloatChannel, keys [[0,0],[20.535,691.2696],[27,-713.3414],[44.997,0]]}
          Translation.Y {type=MovieSceneScriptingFloatChannel}
          Angle {type=MovieSceneScriptingFloatChannel}
          Scale.X {type=MovieSceneScriptingFloatChannel}
          Scale.Y {type=MovieSceneScriptingFloatChannel}
          Shear.X {type=MovieSceneScriptingFloatChannel}
          Shear.Y {type=MovieSceneScriptingFloatChannel}
         ]
        }]
      }
      {
       type=MovieSceneFloatTrack
       display_name=Render Opacity
       sections [{
         type=MovieSceneFloatSection
         end_frame=45
         start_frame=0
         channels [None {type=MovieSceneScriptingFloatChannel, keys [[0,1],[20.535,0],[27,0],[44.997,1]]}]
        }]
      }
      {
       type=MovieSceneFloatVectorTrack
       display_name=Pivot
       sections [{
         type=MovieSceneFloatVectorSection
         end_frame=45
         start_frame=0
         channels [
          Vector.X {type=MovieSceneScriptingFloatChannel}
          Vector.Y {type=MovieSceneScriptingFloatChannel}
         ]
        }]
      }
     ]
    }
    Image_TipBG {
     id=#4
     possessed_class=Image
     tracks [
      {
       type=MovieScene2DTransformTrack
       sections [{
         type=MovieScene2DTransformSection
         end_frame=45
         start_frame=0
         channels [
          Translation.X {type=MovieSceneScriptingFloatChannel}
          Translation.Y {type=MovieSceneScriptingFloatChannel}
          Angle {type=MovieSceneScriptingFloatChannel}
          Scale.X {type=MovieSceneScriptingFloatChannel, keys [[0,1],[20.535,0],[27,0],[44.999,1]]}
          Scale.Y {type=MovieSceneScriptingFloatChannel}
          Shear.X {type=MovieSceneScriptingFloatChannel}
          Shear.Y {type=MovieSceneScriptingFloatChannel}
         ]
        }]
      }
      {
       type=MovieSceneFloatVectorTrack
       display_name=Pivot
       sections [{
         type=MovieSceneFloatVectorSection
         end_frame=44
         start_frame=0
         channels [
          Vector.X {type=MovieSceneScriptingFloatChannel, keys [[0,1],[20.535,1],[27,0],[44.997,0]]}
          Vector.Y {type=MovieSceneScriptingFloatChannel}
         ]
        }]
      }
     ]
    }
    Image_DiamondMark {
     id=#5
     possessed_class=Image
     tracks [
      {
       type=MovieSceneFloatTrack
       display_name=Render Opacity
       sections [{
         type=MovieSceneFloatSection
         end_frame=45
         start_frame=0
         channels [None {type=MovieSceneScriptingFloatChannel, keys [[0,1],[4.577,0],[27,0],[44.997,1]]}]
        }]
      }
      {
       type=MovieScene2DTransformTrack
       sections [{
         type=MovieScene2DTransformSection
         end_frame=44
         start_frame=0
         channels [
          Translation.X {type=MovieSceneScriptingFloatChannel, keys [[0,0],[4.577,0],[27,-506.5034],[44.997,0]]}
          Translation.Y {type=MovieSceneScriptingFloatChannel}
          Angle {type=MovieSceneScriptingFloatChannel}
          Scale.X {type=MovieSceneScriptingFloatChannel}
          Scale.Y {type=MovieSceneScriptingFloatChannel}
          Shear.X {type=MovieSceneScriptingFloatChannel}
          Shear.Y {type=MovieSceneScriptingFloatChannel}
         ]
        }]
      }
     ]
    }
   ]
   display_rate {numerator=60}
   tracks [{type=MovieSceneEventTrack, sections [{type=MovieSceneEventSection}]}]
  }
 }
 SlideTransition_Duplicate {
  binding_count=1
  display_label=""
  length_seconds=0.75
  object=&self:SlideTransition_Duplicate
  sequence {
   bindings [Image_Loading_Incoming {
     id=#1
     possessed_class=Image
     tracks [{
       type=MovieSceneFloatTrack
       display_name=Render Opacity
       sections [{type=MovieSceneFloatSection, end_frame=45, start_frame=0, channels [None {type=MovieSceneScriptingFloatChannel, keys [[0,0],[44.999,1]]}]}]
      }]
    }]
   display_rate {numerator=60}
   tracks [{type=MovieSceneEventTrack, sections [{type=MovieSceneEventSection}]}]
  }
 }
 Intro {
  binding_count=1
  length_seconds=0.45
  object=&self:Intro
  sequence {bindings [Image_Fadeout {
     id=#6
     possessed_class=Image
     tracks [{
       type=MovieSceneFloatTrack
       display_name=Render Opacity
       sections [{type=MovieSceneFloatSection, end_frame=9, start_frame=0, channels [None {type=MovieSceneScriptingFloatChannel, keys [[0,1],[9,0]]}]}]
      }]
    }]}
 }
]
graphs [
 EventGraph {nodes [
   Event Transition() {is=title K2Node_CustomEvent}
   > PlayAnimation(InAnimation=Get TipTransition() -> (TipTransition:Widget Animation Object Reference))
   > SetBrush(
    self=Get Image_Loading_Bottom() -> (Image_Loading_Bottom:Image Object Reference)
    InBrush="Get (a copy)"(
     Array:Array of Slate Brush Structures=Get Slides() -> (Slides:Array of Slate Brush Structures)
     Dimension 1=Get SlidePosition() -> (SlidePosition:int=0)
    ) -> (Output:Slate Brush Structure)
   )
   > n1 = Increment Int(Value=Get SlidePosition() -> (SlidePosition=0))
   n2 = Get Tips() -> (Tips:Array of Texts)
   > n3 = Branch(Condition="integer >= integer"!(A:int=@n1, B:int=Length(TargetArray:Array of Texts=@n2)) -> (ReturnValue:bool) {is=title K2Node_CallFunction})
   > n4 = Set SlidePosition(SlidePosition:int=0) -> (Output_Get:int=0)
   SetBrush(
    execute=[@n4.then, @n3.else]
    self=Get Image_Loading_Incoming() -> (Image_Loading_Incoming:Image Object Reference)
    InBrush="Get (a copy)"(Array=Get Slides(), Dimension 1=Get SlidePosition() -> (SlidePosition=0))
   )
   n5 = Get Brush!(self:Image Object Reference=Get Image_Loading_Incoming()) -> (Brush:Slate Brush Structure)
   n6 = Get Brush(self=Get Image_Loading_Bottom())
   > PlayAnimation(InAnimation=Select(
     Option 0:Widget Animation Object Reference=Get SlideTransition() -> (SlideTransition:Widget Animation Object Reference)
     Option 1:Widget Animation Object Reference=Get SlideTransition_Duplicate() -> (SlideTransition_Duplicate:Widget Animation Object Reference)
     Index="Equal (SlateBrush)"!(A:"Slate Brush Structure (by ref)"=@n5, B:"Slate Brush Structure (by ref)"=@n6) -> (ReturnValue:bool)
    ) -> (ReturnValue:Widget Animation Object Reference))
   n7 = TransitionFinished() {is=title K2Node_CustomEvent}
   > PlayAnimation(InAnimation=Get Idle() -> (Idle:Widget Animation Object Reference))
   n8 = GetTIp!() -> (OutputDelegate, then, SlidePosition:int) {is=title K2Node_CustomEvent}
   > SetBrush(
    self=Get Image_DiamondMark() -> (Image_DiamondMark:Image Object Reference)
    InBrush="Get (a copy)"(Array=Get IconBrush() -> (IconBrush:Array of Slate Brush Structures), Dimension 1=@n8.SlidePosition)
   )
   n9 = "Get (a copy)"(Array:Array of Texts=Get TipTitles() -> (TipTitles:Array of Texts), Dimension 1=@n8.SlidePosition) -> (Output:Text)
   > SetText(
    self=Get RichTextBlock_Title() -> (RichTextBlock_Title:Rich Text Block Object Reference)
    InText=Format Text!(Format="LOCTEXT(\"<title>{0}</>\")", 0:Text=TextToUpper!(InText:"Text (by ref)"=@n9) -> (ReturnValue:Text)) -> (Result)
   )
   > SetVisibility(
    self=Get SizeBox_Title() -> (SizeBox_Title:Size Box Object Reference)
    InVisibility=Select(Option 0:ESlateVisibility Enum=HitTestInvisible, Option 1:ESlateVisibility Enum=Collapsed, Index=TextIsEmpty(InText=@n9)) -> (ReturnValue:ESlateVisibility Enum)
   )
   n10 = "Get (a copy)"(Array=Get Tips(), Dimension 1=@n8.SlidePosition)
   > SetText(self=Get RichTextBlock_Tip() -> (RichTextBlock_Tip:Rich Text Block Object Reference), InText=TextToUpper(InText=@n10))
   > n11 = Branch(Condition=TextIsEmpty(InText=@n10))
   n12 = Get SizeBox_Tip() -> (SizeBox_Tip:Size Box Object Reference)
   > n13 = SetHeightOverride(self=@n12, InHeightOverride=0)
   n14 = SlotAsBorderSlot!(Widget:Widget Object Reference=Get Border_TipContainer() -> (Border_TipContainer:Border Object Reference)) -> (ReturnValue:Border Slot Object Reference)
   SetVerticalAlignment(
    execute=[@n13.then, @n17.then]
    self:Border Slot Object Reference=@n14
    InVerticalAlignment="Get (a copy)"(
     Array:Array of EVerticalAlignment Enums=Get SlideVAlignment() -> (SlideVAlignment:Array of EVerticalAlignment Enums=VAlign_Fill)
     Dimension 1=@n8.SlidePosition
    ) -> (Output:EVerticalAlignment Enum)
   )
   > SetHorizontalAlignment(
    self:Border Slot Object Reference=@n14
    InHorizontalAlignment="Get (a copy)"(
     Array:Array of EHorizontalAlignment Enums=Get SlideHAlignment() -> (SlideHAlignment:Array of EHorizontalAlignment Enums=HAlign_Fill)
     Dimension 1=@n8.SlidePosition
    ) -> (Output:EHorizontalAlignment Enum)
   )
   n15 = TextIsEmpty(InText=@n9)
   n16 = TextIsEmpty(InText=@n10)
   > SetVisibility(self=Get Border_TipContainer(), InVisibility=Select(Option 0=HitTestInvisible, Option 1=Collapsed, Index=AND Boolean(A=@n16, B=@n15)))
   n17 = ClearHeightOverride(execute=@n11.else, self:Size Box Object Reference=@n12)
   GetNextTip() {is=title K2Node_CustomEvent}
   > GetTIp!(execute, self, SlidePosition:int=Get SlidePosition() -> (SlidePosition=0)) -> (then) {is=title K2Node_CallFunction}
   EventIntro() {is=title K2Node_CustomEvent}
   > Call Intro Dispatch!(execute:Exec, self:Heist Loading Object Reference) -> (then:Exec) {is=title K2Node_CallDelegate}
   > PlayAnimation(InAnimation=Get Intro())
   EventAddTitle() {is=title K2Node_CustomEvent}
   n18 = Get CustomWidgetContent() -> (CustomWidgetContent:User Widget Class Reference)
   > Branch(Condition=IsValidClass!(Class:Object Class Reference=@n18) -> (ReturnValue:bool))
   > n19 = Create Widget!(execute:Exec, Class:User Widget Class Reference=@n18, OwningPlayer:Player Controller Object Reference=GetOwningPlayer()) -> (then:Exec, ReturnValue:User Widget Object Reference)
   n20 = Get Overlay_OptionalContent() -> (Overlay_OptionalContent:Overlay Object Reference)
   > ClearChildren(self=@n20)
   > n21 = AddChild(self=@n20, Content=@n19)
   > n22 = Cast To OverlaySlot!(execute, Object=@n21) -> (then, CastFailed, AsOverlay Slot:Overlay Slot Object Reference) {is=title K2Node_DynamicCast}
   > SetHorizontalAlignment(self:Overlay Slot Object Reference=@n22, InHorizontalAlignment=HAlign_Fill)
   > SetVerticalAlignment(self:Overlay Slot Object Reference=@n22, InVerticalAlignment=VAlign_Fill)
   n23 = Event PreConstruct()
   > n24 = Branch(Condition=@n23.IsDesignTime)
   > n25 = EventAddTitle() {is=title K2Node_CallFunction}
   n26 = Get SlidePosition() -> (SlidePosition=0)
   SetBrush(
    execute=[@n25.then, @n28.then]
    self=Get Image_Loading_Bottom()
    InBrush="Get (a copy)"(Array:Array of Slate Brush Structures=Get Slides(), Dimension 1=@n26) -> (Output:Slate Brush Structure)
   )
   > GetTIp(SlidePosition=@n26)
   > SetBrush(self=Get Image_TipBG() -> (Image_TipBG:Image Object Reference), InBrush=Get TipBGBrush() -> (TipBGBrush:Slate Brush Structure))
   > SetTextStyleSet!(
    execute
    self:Rich Text Block Object Reference=[Get RichTextBlock_Tip(), Get RichTextBlock_Title()]
    NewTextStyleSet:Data Table Object Reference=Get TextStyles() -> (TextStyles:Data Table Object Reference)
   ) -> (then)
   > SetUserSpecifiedScale!(
    execute
    self:Scale Box Object Reference=Get ScaleBox_TipScale() -> (ScaleBox_TipScale:Scale Box Object Reference)
    InUserSpecifiedScale:float=Get TipScale() -> (TipScale:double)
   ) -> (then)
   n27 = Length(TargetArray=Get Tips())
   n28 = Set SlidePosition(
    execute=@n24.else
    SlidePosition=RandomIntegerInRange!(Min:int=0, Max:int=int - int(A=@n27, B=1) {is=title K2Node_CallFunction}) -> (ReturnValue:int=0)
   ) -> (Output_Get=0)
   Event Construct()
   n29 = Get Idle()
   > BindToAnimationFinished(Animation=@n29, Delegate=Create Event())
   > PlayAnimation(InAnimation=@n29)
   > BindToAnimationFinished!(
    execute
    self:Widget Animation Object Reference=[Get SlideTransition(), Get SlideTransition_Duplicate()]
    Widget:User Widget Object Reference=Self-Reference()
    Delegate=@n7
   ) -> (then)
   > EventAddTitle()
   > EventIntro() {is=title K2Node_CallFunction}
  ]}
 IntroDispatch {nodes [IntroDispatch() {is=title K2Node_FunctionEntry}]}
]
variables [
 CustomWidgetContent {type=Class, default=null}
 IconBrush {
  type="array<SlateBrush>"
  default [
   {
    bIsDynamicallyLoaded=false
    drawAs=Image
    imageType=NoImage
    mirroring=NoMirror
    resourceName=None
    resourceObject=/Game/Athena/UI/DailyQuests/Textures/M_Heist_DiamondMark_Inst
    tiling=NoTile
    imageSize=<88.1837 60.3204>
    margin=<ltrb 0 0 0 0>
    outlineSettings {
     bUseBrushTransparency=false
     roundingType=HalfHeightRadius
     width=0
     color {colorUseRule=UseColor_Specified, specifiedColor=<rgba 0 0 0 0>}
     cornerRadii=<0 0 0 1>
    }
    tintColor {colorUseRule=UseColor_Specified, specifiedColor=<rgba 1 1 1 1>}
    uVRegion {bIsValid=false, max=<0 0>, min=<0 0>}
   }
   {
    bIsDynamicallyLoaded=false
    drawAs=Image
    imageType=NoImage
    mirroring=NoMirror
    resourceName=None
    resourceObject=&M_Heist_DiamondMark_Inst
    tiling=NoTile
    imageSize=<88.1837 60.3204>
    margin=<ltrb 0 0 0 0>
    outlineSettings {
     bUseBrushTransparency=false
     roundingType=HalfHeightRadius
     width=0
     color {colorUseRule=UseColor_Specified, specifiedColor=<rgba 0 0 0 0>}
     cornerRadii=<0 0 0 1>
    }
    tintColor {colorUseRule=UseColor_Specified, specifiedColor=<rgba 1 1 1 1>}
    uVRegion {bIsValid=false, max=<0 0>, min=<0 0>}
   }
  ]
 }
 IntroDispatch {}
 SlideHAlignment {type="array<EHorizontalAlignment>", default [HAlign_Right,HAlign_Left]}
 SlidePosition {type=integer}
 Slides {
  type="array<SlateBrush>"
  default [
   {
    bIsDynamicallyLoaded=false
    drawAs=Image
    imageType=NoImage
    mirroring=NoMirror
    resourceName=None
    resourceObject=/Game/UI/Foundation/Textures/LoadingScreens/LTM/Materials/T_UI_HeistLoading1
    tiling=NoTile
    imageSize=<1920 1080>
    margin=<ltrb 0 0 0 0>
    outlineSettings {
     bUseBrushTransparency=false
     roundingType=HalfHeightRadius
     width=0
     color {colorUseRule=UseColor_Specified, specifiedColor=<rgba 0 0 0 0>}
     cornerRadii=<0 0 0 1>
    }
    tintColor {colorUseRule=UseColor_Specified, specifiedColor=<rgba 1 1 1 1>}
    uVRegion {bIsValid=false, max=<0 0>, min=<0 0>}
   }
   {
    bIsDynamicallyLoaded=false
    drawAs=Image
    imageType=NoImage
    mirroring=NoMirror
    resourceName=None
    resourceObject=/Game/UI/Foundation/Textures/LoadingScreens/LTM/Materials/T_UI_HeistLoading2
    tiling=NoTile
    imageSize=<1920 1080>
    margin=<ltrb 0 0 0 0>
    outlineSettings {
     bUseBrushTransparency=false
     roundingType=HalfHeightRadius
     width=0
     color {colorUseRule=UseColor_Specified, specifiedColor=<rgba 0 0 0 0>}
     cornerRadii=<0 0 0 1>
    }
    tintColor {colorUseRule=UseColor_Specified, specifiedColor=<rgba 1 1 1 1>}
    uVRegion {bIsValid=false, max=<0 0>, min=<0 0>}
   }
  ]
 }
 SlideVAlignment {type="array<EVerticalAlignment>", default [VAlign_Bottom, VAlign_Bottom]}
 TextStyles {type=DataTable, default=&/AthenaHeistLoadingStyles}
 TipBGBrush {
  type=SlateBrush
  default {
   bIsDynamicallyLoaded=false
   drawAs=Image
   imageType=NoImage
   mirroring=NoMirror
   resourceName=None
   resourceObject=null
   tiling=NoTile
   imageSize=<32 32>
   margin=<ltrb 0 0 0 0>
   outlineSettings {
    bUseBrushTransparency=false
    roundingType=HalfHeightRadius
    width=0
    color {colorUseRule=UseColor_Specified, specifiedColor=<rgba 0 0 0 0>}
    cornerRadii=<0 0 0 1>
   }
   tintColor {colorUseRule=UseColor_Specified, specifiedColor=<rgba 0.0168 0.0168 0.0168 1>}
   uVRegion {bIsValid=false, max=<0 0>, min=<0 0>}
  }
 }
 Tips {
  type="array<string>"
  default [
   "LOCATE <b>JEWELS</> HIDDEN IN <b>SAFES</> \r\nDROPPED RANDOMLY ACROSS THE MAP."
   "REACH A <b>VAN</> WITH A <b>JEWEL</> TO WIN! THERE ARE FOUR <b>JEWELS</> AND FOUR <b>VANS</>, SO KEEP FIGHTING!"
  ]
 }
 TipScale {type=number, default=1}
 TipTitles {type="array<string>", default [THE SETUP, THE GETAWAY]}
]
widgets {
 info {inheritedWidgetCount=0, parentClass=&UserWidget, rootWidgetClass=&Overlay, widgetCount=23}
 widgets [
  Overlay_0 {widgetClassPath=&Overlay, widget=Overlay_0, properties {clipping=ClipToBounds}}
  Image_FlatBG {
   widgetClassPath=&Image
   bIsVariable=true
   parent=Overlay_0
   widget=Image_FlatBG
   properties {colorAndOpacity=<rgba 0 0 0 1>}
   slot {class=OverlaySlot, properties {horizontalAlignment=HAlign_Fill, verticalAlignment=VAlign_Fill}}
  }
  ScaleBox_0 {
   widgetClassPath=&ScaleBox
   parent=Overlay_0
   widget=ScaleBox_0
   properties {stretch=ScaleToFitY, stretchDirection=UpOnly}
   slot {class=OverlaySlot, properties {horizontalAlignment=HAlign_Fill, verticalAlignment=VAlign_Fill}}
  }
  Overlay_1 {widgetClassPath=&Overlay, parent=ScaleBox_0, widget=Overlay_1, slot {class=ScaleBoxSlot}}
  Image_Loading_Bottom {
   widgetClassPath=&Image
   bIsVariable=true
   parent=Overlay_1
   widget=Image_Loading_Bottom
   properties {brush {
     bIsDynamicallyLoaded=false
     drawAs=Image
     imageType=NoImage
     mirroring=NoMirror
     resourceName=None
     resourceObject=&T_UI_HeistLoading1
     tiling=NoTile
     imageSize=<1920 1080>
     margin=<ltrb 0 0 0 0>
     outlineSettings {
      bUseBrushTransparency=false
      roundingType=HalfHeightRadius
      width=0
      color {colorUseRule=UseColor_Specified, specifiedColor=<rgba 0 0 0 0>}
      cornerRadii=<0 0 0 0>
     }
     tintColor {colorUseRule=UseColor_Specified, specifiedColor=<rgba 1 1 1 1>}
     uVRegion {bIsValid=false, max=<0 0>, min=<0 0>}
    }}
   slot {class=OverlaySlot, properties {horizontalAlignment=HAlign_Fill, verticalAlignment=VAlign_Fill}}
  }
  Image_Loading_Incoming {
   widgetClassPath=&Image
   bIsVariable=true
   parent=Overlay_1
   widget=Image_Loading_Incoming
   properties {
    renderOpacity=0
    brush {
     bIsDynamicallyLoaded=false
     drawAs=Image
     imageType=NoImage
     mirroring=NoMirror
     resourceName=None
     resourceObject=&T_UI_HeistLoading2
     tiling=NoTile
     imageSize=<1920 1080>
     margin=<ltrb 0 0 0 0>
     outlineSettings {
      bUseBrushTransparency=false
      roundingType=HalfHeightRadius
      width=0
      color {colorUseRule=UseColor_Specified, specifiedColor=<rgba 0 0 0 0>}
      cornerRadii=<0 0 0 0>
     }
     tintColor {colorUseRule=UseColor_Specified, specifiedColor=<rgba 1 1 1 1>}
     uVRegion {bIsValid=false, max=<0 0>, min=<0 0>}
    }
    renderTransformPivot=<0.5 0.3>
   }
   slot {class=OverlaySlot, properties {horizontalAlignment=HAlign_Fill, verticalAlignment=VAlign_Fill}}
  }
  SafeZone_0 {
   widgetClassPath=&SafeZone
   bIsVariable=true
   parent=Overlay_0
   widget=SafeZone_0
   slot {class=OverlaySlot, properties {horizontalAlignment=HAlign_Fill, verticalAlignment=VAlign_Fill}}
  }
  Overlay_2 {widgetClassPath=&Overlay, parent=SafeZone_0, widget=Overlay_2, slot {class=SafeZoneSlot}}
  CommonTextBlock_1 {
   widgetClassPath=&CommonTextBlock
   parent=Overlay_2
   widget=CommonTextBlock_1
   properties {
    text=EARLY ACCESS
    visibility=Collapsed
    font {
     bForceMonospaced=false
     bMaterialIsStencil=false
     fontMaterial=null
     fontObject=&BurbankSmall
     letterSpacing=0
     monospacedWidth=1
     size=24
     skewAmount=0
     typefaceFontName=NotoSans-Regular
     outlineSettings {
      bApplyOutlineToDropShadows=false
      bMiteredCorners=false
      bSeparateFillAlpha=false
      outlineMaterial=null
      outlineSize=1
      outlineColor=<rgba 0.0052 0.0052 0.0056 0.2>
     }
    }
    shadowColorAndOpacity=<rgba 0.0052 0.0052 0.0056 0.2>
    shadowOffset=<3 3>
   }
   slot {class=OverlaySlot, properties {horizontalAlignment=HAlign_Right, verticalAlignment=VAlign_Bottom, padding=<ltrb 20 20 20 20>}}
  }
  Border_ContentPadding {
   widgetClassPath=&Border
   bIsVariable=true
   parent=Overlay_2
   widget=Border_ContentPadding
   properties {
    horizontalAlignment=HAlign_Right
    verticalAlignment=VAlign_Bottom
    background {
     bIsDynamicallyLoaded=false
     drawAs=NoDrawType
     imageType=NoImage
     mirroring=NoMirror
     resourceName=None
     resourceObject=null
     tiling=NoTile
     imageSize=<32 32>
     margin=<ltrb 0 0 0 0>
     outlineSettings {
      bUseBrushTransparency=false
      roundingType=HalfHeightRadius
      width=0
      color {colorUseRule=UseColor_Specified, specifiedColor=<rgba 0 0 0 0>}
      cornerRadii=<0 0 0 0>
     }
     tintColor {colorUseRule=UseColor_Specified, specifiedColor=<rgba 1 1 1 1>}
     uVRegion {bIsValid=false, max=<0 0>, min=<0 0>}
    }
    padding=<ltrb 80 50 80 125>
   }
   slot {class=OverlaySlot, properties {horizontalAlignment=HAlign_Fill, verticalAlignment=VAlign_Fill}}
  }
  Border_TipContainer {
   widgetClassPath=&Border
   bIsVariable=true
   parent=Border_ContentPadding
   widget=Border_TipContainer
   properties {
    background {
     bIsDynamicallyLoaded=false
     drawAs=NoDrawType
     imageType=NoImage
     mirroring=NoMirror
     resourceName=None
     resourceObject=null
     tiling=NoTile
     imageSize=<32 32>
     margin=<ltrb 0 0 0 0>
     outlineSettings {
      bUseBrushTransparency=false
      roundingType=HalfHeightRadius
      width=0
      color {colorUseRule=UseColor_Specified, specifiedColor=<rgba 0 0 0 0>}
      cornerRadii=<0 0 0 0>
     }
     tintColor {colorUseRule=UseColor_Specified, specifiedColor=<rgba 1 1 1 1>}
     uVRegion {bIsValid=false, max=<0 0>, min=<0 0>}
    }
    padding=<ltrb 0 0 0 0>
   }
   slot {class=BorderSlot, properties {horizontalAlignment=HAlign_Right, verticalAlignment=VAlign_Bottom, padding=<ltrb 80 50 80 125>}}
  }
  ScaleBox_TipScale {
   widgetClassPath=&ScaleBox
   bIsVariable=true
   parent=Border_TipContainer
   widget=ScaleBox_TipScale
   properties {stretch=UserSpecified}
   slot {class=BorderSlot, properties {padding=<ltrb 0 0 0 0>}}
  }
  Overlay_4 {
   widgetClassPath=&Overlay
   parent=ScaleBox_TipScale
   widget=Overlay_4
   properties {clipping=ClipToBounds, renderTransform {angle=0, scale=<1 1>, shear=<-14 0>, translation=<0 0>}}
   slot {class=ScaleBoxSlot}
  }
  Image_TipBG {
   widgetClassPath=&Image
   bIsVariable=true
   parent=Overlay_4
   widget=Image_TipBG
   slot {class=OverlaySlot, properties {horizontalAlignment=HAlign_Fill, verticalAlignment=VAlign_Fill}}
  }
  Image_DiamondMark {
   widgetClassPath=&Image
   bIsVariable=true
   parent=Overlay_4
   widget=Image_DiamondMark
   properties {
    brush {
     bIsDynamicallyLoaded=false
     drawAs=Image
     imageType=NoImage
     mirroring=NoMirror
     resourceName=None
     resourceObject=&M_Heist_DiamondMark_Inst
     tiling=NoTile
     imageSize=<88.1837 60.3204>
     margin=<ltrb 0 0 0 0>
     outlineSettings {
      bUseBrushTransparency=false
      roundingType=HalfHeightRadius
      width=0
      color {colorUseRule=UseColor_Specified, specifiedColor=<rgba 0 0 0 0>}
      cornerRadii=<0 0 0 0>
     }
     tintColor {colorUseRule=UseColor_Specified, specifiedColor=<rgba 1 1 1 1>}
     uVRegion {bIsValid=false, max=<0 0>, min=<0 0>}
    }
    renderTransform {angle=0, scale=<1 1>, shear=<14 0>, translation=<0 0>}
   }
   slot {class=OverlaySlot, properties {horizontalAlignment=HAlign_Right, padding=<ltrb 20 20 20 0>}}
  }
  VerticalBox_0 {
   widgetClassPath=&VerticalBox
   parent=Overlay_4
   widget=VerticalBox_0
   properties {renderTransform {angle=0, scale=<1 1>, shear=<14 0>, translation=<0 0>}}
   slot {class=OverlaySlot, properties {padding=<ltrb 70 20 70 40>}}
  }
  SizeBox_Title {
   widgetClassPath=&SizeBox
   bIsVariable=true
   parent=VerticalBox_0
   widget=SizeBox_Title
   properties {bOverride_WidthOverride=true}
   slot {class=VerticalBoxSlot, properties {verticalAlignment=VAlign_Bottom, padding=<ltrb 0 0 20 0>}}
  }
  ScaleBox_1 {
   widgetClassPath=&ScaleBox
   parent=SizeBox_Title
   widget=ScaleBox_1
   properties {stretchDirection=DownOnly}
   slot {class=SizeBoxSlot, properties {horizontalAlignment=HAlign_Left, verticalAlignment=VAlign_Bottom}}
  }
  RichTextBlock_Title {
   widgetClassPath=/Script/UMG.RichTextBlock
   bIsVariable=true
   parent=ScaleBox_1
   widget=RichTextBlock_Title
   properties {
    text="<title>THE SETUP</>"
    textStyleSet=&AthenaHeistLoadingStyles
    renderTransform {angle=0, scale=<1 1>, shear=<-7 0>, translation=<0 0>}
   }
   slot {class=ScaleBoxSlot, properties {horizontalAlignment=HAlign_Left, verticalAlignment=VAlign_Bottom}}
  }
  SizeBox_Tip {
   widgetClassPath=&SizeBox
   bIsVariable=true
   parent=VerticalBox_0
   widget=SizeBox_Tip
   properties {bOverride_MinDesiredWidth=true, minDesiredWidth=675}
   slot {class=VerticalBoxSlot, properties {padding=<ltrb -10 0 0 0>}}
  }
  RichTextBlock_Tip {
   widgetClassPath=&RichTextBlock
   bIsVariable=true
   parent=SizeBox_Tip
   widget=RichTextBlock_Tip
   properties {
    lineHeightPercentage=1.2
    text="LOCATE <b>JEWELS</> HIDDEN IN <b>SAFES</> DROPPED RANDOMLY ACROSS THE MAP."
    textStyleSet=&/AthenaAshtonLoadingStyles
    wrapTextAt=685
   }
   slot {class=SizeBoxSlot}
  }
  Overlay_OptionalContent {
   widgetClassPath=&Overlay
   bIsVariable=true
   parent=Overlay_2
   widget=Overlay_OptionalContent
   slot {class=OverlaySlot, properties {horizontalAlignment=HAlign_Fill, verticalAlignment=VAlign_Fill}}
  }
  Image_Fadeout {
   widgetClassPath=&Image
   parent=Overlay_0
   widget=Image_Fadeout
   properties {renderOpacity=0, colorAndOpacity=<rgba 0.0052 0.0052 0.0056 1>}
   slot {class=OverlaySlot, properties {horizontalAlignment=HAlign_Fill, verticalAlignment=VAlign_Fill}}
  }
 ]
}
