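# UI material that draws a skewed box with a border, stored with the LTM loading screen assets.
# Four rotatable SmoothStep edges mask each of the outer and inner quads, with InnerColor alpha controlling how hollow the center is.
# The border blends Color1 to Color2 along a rotated axis, or a vertical power gradient overrides both border and fill color.
# Static switches optionally composite a tinted icon and add a shine swipe.
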Material /Game/UI/Foundation/Textures/LoadingScreens/LTM/Materials/M_UI_SkewBoxBorder
graph {
 properties {
  bUsedWithStaticMesh=false
  blendMode=BLEND_Translucent
  materialDomain=MD_UI
  pixelDepthOffsetMode=PDOM_Legacy
  usedShadingModels=Default Lit
 }
 nodes [
  n1 = VectorParameter() {parameterName=Color1, defaultValue=<rgba 1 1 1 1>}
  n2 = ScalarParameter() {defaultValue=0.1, parameterName=LeftEdge}
  n3 = MaterialFunctionCall(
   UVs=TextureCoordinate()
   Rotation Center=AppendVector(A=@n2, B=Constant() {r=0.5})
   "Rotation Angle (0-1)"=ScalarParameter() {parameterName=LeftRotation, sliderMax=0.25, sliderMin=-0.25}
  ) {materialFunction=&CustomRotator}
  n4 = ScalarParameter() {defaultValue=0.001, parameterName=EdgeSoftness}
  n5 = OneMinus(None=MaterialFunctionCall(Alpha=ComponentMask(None=@n3.Rotated Values) {r=true}, Min=@n2, Max=Add(A=@n2, B=@n4)) {materialFunction=&SmoothStep}.Result)
  n6 = TextureCoordinate()
  n7 = OneMinus(None=ScalarParameter() {defaultValue=0.1, parameterName=RightEdge})
  n8 = AppendVector(A=@n7, B=Constant() {r=0.5})
  n9 = ScalarParameter() {parameterName=RightRotation, sliderMax=0.25, sliderMin=-0.25}
  n10 = MaterialFunctionCall(
   Alpha=ComponentMask(None=MaterialFunctionCall(UVs=@n6, Rotation Center=@n8, "Rotation Angle (0-1)"=@n9) {materialFunction=&CustomRotator}.Rotated Values) {r=true}
   Min=@n7
   Max=Add(A=@n7, B=ScalarParameter() {defaultValue=0.001, parameterName=EdgeSoftness})
  ) {materialFunction=&SmoothStep}
  n11 = OneMinus(None=ScalarParameter() {defaultValue=0.1, parameterName=BottomEdge})
  n12 = Constant() {r=0.5}
  n13 = ComponentMask(None=MaterialFunctionCall(
    UVs=TextureCoordinate()
    Rotation Center=AppendVector(A=@n11, B=@n12)
    "Rotation Angle (0-1)"=ScalarParameter() {parameterName=BottomRotation, sliderMax=0.25, sliderMin=-0.25}
   ) {materialFunction=&CustomRotator}.Rotated Values) {g=true}
  n14 = Add(A=@n11, B=ScalarParameter() {defaultValue=0.001, parameterName=EdgeSoftness})
  n15 = ScalarParameter() {defaultValue=0.1, parameterName=TopEdge}
  n16 = Constant() {r=0.5}
  n17 = ComponentMask(None=MaterialFunctionCall(
    UVs=TextureCoordinate()
    Rotation Center=AppendVector(A=@n15, B=@n16)
    "Rotation Angle (0-1)"=ScalarParameter() {parameterName=TopRotation, sliderMax=0.25, sliderMin=-0.25}
   ) {materialFunction=&CustomRotator}.Rotated Values) {g=true}
  n18 = Add(A=@n15, B=ScalarParameter() {defaultValue=0.001, parameterName=EdgeSoftness})
  n19 = Max(
   A=Max(
    A=Max(A=@n5, B=@n10.Result)
    B=MaterialFunctionCall(Alpha=@n13, Min=@n11, Max=@n14) {materialFunction=&SmoothStep}.Result
   )
   B=OneMinus(None=MaterialFunctionCall(Alpha=@n17, Min=@n15, Max=@n18) {materialFunction=&SmoothStep}.Result)
  )
  n20 = ScalarParameter() {defaultValue=0.12, group=Inner, parameterName=LeftEdge_1}
  n21 = MaterialFunctionCall(
   UVs=TextureCoordinate()
   Rotation Center=AppendVector(A=@n20, B=Constant() {r=0.5})
   "Rotation Angle (0-1)"=ScalarParameter() {group=Inner, parameterName=LeftRotation_1, sliderMax=0.25, sliderMin=-0.25}
  ) {materialFunction=&CustomRotator}
  n22 = ScalarParameter() {defaultValue=0.001, group=Inner, parameterName=EdgeSoftness}
  n23 = OneMinus(None=MaterialFunctionCall(Alpha=ComponentMask(None=@n21.Rotated Values) {r=true}, Min=@n20, Max=Add(A=@n20, B=@n22)) {materialFunction=&SmoothStep}.Result)
  n24 = TextureCoordinate()
  n25 = OneMinus(None=ScalarParameter() {defaultValue=0.12, group=Inner, parameterName=RightEdge_1})
  n26 = AppendVector(A=@n25, B=Constant() {r=0.5})
  n27 = ScalarParameter() {group=Inner, parameterName=RightRotation_1, sliderMax=0.25, sliderMin=-0.25}
  n28 = MaterialFunctionCall(
   Alpha=ComponentMask(None=MaterialFunctionCall(UVs=@n24, Rotation Center=@n26, "Rotation Angle (0-1)"=@n27) {materialFunction=&CustomRotator}.Rotated Values) {r=true}
   Min=@n25
   Max=Add(A=@n25, B=ScalarParameter() {defaultValue=0.001, group=Inner, parameterName=EdgeSoftness})
  ) {materialFunction=&SmoothStep}
  n29 = OneMinus(None=ScalarParameter() {defaultValue=0.12, group=Inner, parameterName=BottomEdge_1})
  n30 = Constant() {r=0.5}
  n31 = ComponentMask(None=MaterialFunctionCall(
    UVs=TextureCoordinate()
    Rotation Center=AppendVector(A=@n29, B=@n30)
    "Rotation Angle (0-1)"=ScalarParameter() {group=Inner, parameterName=BottomRotation_1, sliderMax=0.25, sliderMin=-0.25}
   ) {materialFunction=&CustomRotator}.Rotated Values) {g=true}
  n32 = Add(A=@n29, B=ScalarParameter() {defaultValue=0.001, group=Inner, parameterName=EdgeSoftness})
  n33 = ScalarParameter() {defaultValue=0.12, group=Inner, parameterName=TopEdge_1}
  n34 = Constant() {r=0.5}
  n35 = ComponentMask(None=MaterialFunctionCall(
    UVs=TextureCoordinate()
    Rotation Center=AppendVector(A=@n33, B=@n34)
    "Rotation Angle (0-1)"=ScalarParameter() {group=Inner, parameterName=TopRotation_1, sliderMax=0.25, sliderMin=-0.25}
   ) {materialFunction=&CustomRotator}.Rotated Values) {g=true}
  n36 = Add(A=@n33, B=ScalarParameter() {defaultValue=0.001, group=Inner, parameterName=EdgeSoftness})
  n37 = Max(
   A=Max(
    A=Max(A=@n23, B=@n28.Result)
    B=MaterialFunctionCall(Alpha=@n31, Min=@n29, Max=@n32) {materialFunction=&SmoothStep}.Result
   )
   B=OneMinus(None=MaterialFunctionCall(Alpha=@n35, Min=@n33, Max=@n36) {materialFunction=&SmoothStep}.Result)
  )
  n38 = OneMinus(None=@n37)
  n39 = VectorParameter() {parameterName=InnerColor, defaultValue=<rgba 1 1 1 0>}
  n40 = OneMinus(None=@n39.A)
  n41 = Multiply(A=@n1.A, B=Subtract(A=OneMinus(None=@n19), B=Multiply(A=@n38, B=@n40)))
  n42 = TextureSampleParameter2D() {parameterName=IconTexture, texture=/Game/Athena/UI/PostMatch/AthenaChallenge/Textures/T_UI_CheckMark}
  StaticSwitchParameter(True=Clamp(None=Add(A=@n41, B=@n42.A)), False=@n41) {parameterName=UseIconTexture}
  n43 = Multiply(A=VectorParameter() {parameterName=IconTint, defaultValue=<rgba 1 1 1 1>}.RGB, B=@n42.RGB)
  n44 = OneMinus(None=@n42.A)
  n45 = VectorParameter() {parameterName=Color2, defaultValue=<rgba 1 1 1 1>}
  n46 = TextureCoordinate()
  n47 = ComponentMask(None=@n46) {g=true}
  n48 = ScalarParameter() {defaultValue=1, parameterName=GradPow}
  n49 = ComponentMask(None=MaterialFunctionCall(UVs=@n46, "Rotation Angle (0-1)"=ScalarParameter() {defaultValue=0.2, parameterName=Rotation, sliderMax=1}) {materialFunction=&CustomRotator}.Rotated Values) {r=true}
  n50 = StaticSwitchParameter(
   True=LinearInterpolate(A=@n1.RGB, B=@n45.RGB, Alpha=Power(Base=@n47, Exp=@n48))
   False=LinearInterpolate(A=@n39.RGB, B=LinearInterpolate(A=@n1.RGB, B=@n45.RGB, Alpha=@n49), Alpha=@n37)
  ) {parameterName=UseGradientOverride}
  n51 = StaticSwitchParameter(True=Add(A=Multiply(A=@n43, B=@n42.A), B=Multiply(A=@n44, B=@n50)), False=@n50) {parameterName=UseIconTexture}
  StaticSwitchParameter(
   True=LinearInterpolate(
    A=@n51
    B=VectorParameter() {parameterName=ShineColor, defaultValue=<rgba 1 1 1 1>}.RGB
    Alpha=MaterialFunctionCall!(In_Uv) {materialFunction=/Game/UI/Foundation/Materials/M-Function-ShineSwipe}.Alpha
   )
   False=@n51
  ) {parameterName=UseShineSwipe}
 ]
}
